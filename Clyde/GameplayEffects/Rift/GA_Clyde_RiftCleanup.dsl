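; Gameplay ability that cleans up after a Clyde Rift by clearing the hiding-prop interaction block from the player pawn.
; On activation, where cosmetic events can execute, it gets the avatar pawn and checks for the HidingProp BlockInteract tag.
; If present, it removes the loose BlockInteract tags, and it ends the ability whenever the pawn is valid.
; The passive-activation failure handler only calls its parent.
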(event Ability|EventActivateAbility
  (|CallParentFunction)
  (Networking|CanExecuteCosmeticEvents
    (:True
      (bind _outfortplayerpawn (CallFunction|GetAvatarasValidFortPlayerPawn)
        (:ValidFortPlayerPawn
          (if (GameplayTags|HasMatchingGameplayTag _outfortplayerpawn "(TagName=\"Granted.Athena.EnvItem.HidingProp.BlockInteract\")")
            (Ability|GameplayEffect|RemoveLooseGameplayTags _outfortplayerpawn (Variables|Default|GetBlockInteractTag))
            (else
              _))
          (Ability|EndAbility))
        (:AvatarCastFailed
          _)))))

(event Ability|EventFailedToActivatePassiveAbility
  (params (ActorInfo :type Gameplay%20Ability%20Actor%20Info%20Structure))
  (|CallParentFunction ActorInfo))
